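# Additive post-process material for the HyperBloom POI that adds a rippling distortion and a rainbow-like shimmer to the screen.
# Time-animated sine waves, radiating from screen center, scale scene UVs for the distorted, boosted image.
# Two noise-warped, panning polar samples of a phased texture are max-combined and weighted toward screen edges.
# PPBlendAMt blends this over the scene; world-position, camera-direction and center masks are also declared, alongside a disconnected radial-blur branch.
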Material /Game/Environments/Artemis/Sets/Daisy/HyperBloom/Materials/PP_HyperBloomPOI
graph {
 properties {blendMode=BLEND_Additive, materialDomain=MD_PostProcess, pixelDepthOffsetMode=PDOM_Legacy, shadingModel=MSM_Unlit, usedShadingModels=Unlit}
 nodes [
  n1 = ScalarParameter() {defaultValue=1.01, parameterName=DistortAmt}
  n2 = ComponentMask(None=ScreenPosition().ViewportUV) {g=true, r=true}
  n3 = Subtract(A=Multiply(A=@n2) {constB=2})
  n4 = DotProduct(A=@n3, B=@n3)
  n5 = Add(A=OneMinus(None=@n4), B=Multiply(A=Time(), B=ScalarParameter() {defaultValue=0.25, parameterName=TimeSpeed}))
  n6 = Multiply(A=@n5, B=ScalarParameter() {defaultValue=7, parameterName=Frequency})
  n7 = Clamp(None=ConstantBiasScale(None=Sine(None=@n6)))
  n8 = Sine(None=Multiply(A=@n5, B=ScalarParameter() {defaultValue=1, parameterName=MacroFrequency}))
  n9 = MaterialFunctionCall(
   UVs=@n2
   Texture Scale=LinearInterpolate(
    A=LinearInterpolate(B=@n1, Alpha=@n7) {constA=1}
    B=ScalarParameter() {defaultValue=1.02, parameterName=MAcro0Distort}
    Alpha=ConstantBiasScale(None=@n8)
   )
  ) {materialFunction=&ScaleUVsByCenter}
  n10 = SceneTexture(UVs=@n9.UVs) {sceneTextureId=PPI_PostProcessInput0}
  Add(
   A=Multiply(A=MaterialFunctionCall!(IsStaticmesh) {materialFunction=/Game/Packages/Fortnite_UniqueMaterials_1/PostProcess/RadialBlur/RadialBlur}.Result) {constB=0.25}
   B=@n10.Color
  )
  n11 = NamedRerouteDeclaration(None=VectorParameter() {parameterName=WolrdPosCenter, defaultValue=<rgba -12695.293 -79172.5078 4564.2451 1>}.RGB) {name=WorldLocation, nodeColor=<rgba 1 0.6562 0 1>}
  ComponentMask(None=Transform(None=@n11) {transformSourceType=TRANSFORMSOURCE_World, transformType=TRANSFORM_Camera}) {g=true, r=true}
  Constant()
  Constant()
  n12 = Multiply(
   A=SceneTexture(UVs=@n9.UVs) {sceneTextureId=PPI_PostProcessInput0}.Color
   B=ScalarParameter() {defaultValue=0.1, parameterName=PP_AddStrength}
  ) {constB=0.25}
  n13 = Power(Base=@n4) {constExponent=4}
  n14 = Normalize(VectorInput=NamedRerouteDeclaration(None=Add(A=@n2) {constB=-0.5}) {name=ScreenPosUVs, nodeColor=<rgba 0 1 0.6953 1>})
  n15 = Multiply(A=@n14, B=ScalarParameter() {defaultValue=2, parameterName=PolyCount})
  n16 = ComponentMask(None=TextureSample(UVs=Panner(Coordinate=@n15)) {texture=/Game/Characters/Player/Male/Medium/Bodies/M_MED_Viper/Textures/Glitch/T_SoftNoise}.RGB) {g=true, r=true}
  n17 = Multiply(A=Subtract(A=Multiply(A=@n16) {constB=2}), B=ScalarParameter() {defaultValue=0.05, parameterName=Distort_Amt})
  n18 = MaterialFunctionCall!(CoordinateOffset, Coordinate=@n14) {materialFunction=&/MF_PolarCoordinates}
  n19 = MaterialFunctionCall(Float2=@n18.Result) {materialFunction=&BreakOutFloat2Components}
  n20 = ComponentMask(None=MaterialFunctionCall(X=Multiply(A=@n19.R), Y=@n19.G) {materialFunction=&MakeFloat2}.Result) {g=true, r=true}
  n21 = Panner(Coordinate=Multiply(A=@n20, B=Constant2Vector() {g=-1, r=1}) {constB=-1}) {speedX=0.03, speedY=0.03}
  n22 = TextureSampleParameter2D(UVs=Panner(Coordinate=Add(A=@n17, B=@n21)) {bFractionalPart=true, speedX=0.0305, speedY=0.01}) {
   group=Chrome_Phased
   parameterName=PhasedMap
   samplerSource=SSM_Wrap_WorldGroupSettings
   texture=/Game/TechArt/Seasonal/LackLuster/Textures/T_Lackluster_Phased
  }
  n23 = Panner(Coordinate=@n20) {speedX=0.03, speedY=-0.03}
  n24 = TextureSampleParameter2D(UVs=Panner(Coordinate=Add(A=@n17, B=@n23)) {bFractionalPart=true, speedX=0.0225, speedY=0.0065}) {
   group=Chrome_Phased
   parameterName=PhasedMap
   samplerSource=SSM_Wrap_WorldGroupSettings
   texture=&T_Lackluster_Phased
  }
  n25 = AppendVector(A=Max(A=@n22.RGB, B=@n24.RGB), B=Constant())
  n26 = Multiply(A=Multiply(A=Multiply(A=@n13, B=@n25), B=NamedRerouteUsage()), B=ScalarParameter() {defaultValue=0.5, parameterName=RainBowStrength})
  LinearInterpolate(
   A=SceneTexture() {sceneTextureId=PPI_PostProcessInput0}.Color
   B=Add(A=Add(A=@n12, B=@n10.Color), B=@n26)
   Alpha=ScalarParameter() {parameterName=PPBlendAMt}
  )
  MaterialFunctionCall(World Position=@n11) {materialFunction=&TransformToClipSpace}
  MaterialFunctionCall!(Coordinate=@n14) {materialFunction=/Game/Effects/Niagara/MasterMaterials/Functions/UVs/PolarCoordinates}
  Multiply(A=@n18.Result) {constB=4}
  n27 = Constant3Vector() {constant=<rgba -12471.6211 -79543.4375 1065.8462 0>}
  n28 = Normalize(VectorInput=Subtract(A=@n27, B=CameraPositionWS()))
  n29 = MaterialFunctionCall() {materialFunction=&CameraDirectionVector}
  NamedRerouteDeclaration(None=Clamp(None=DotProduct(A=@n28, B=@n29.Result))) {name=CameraDirectionMask, nodeColor=<rgba 0 1 0.5781 1>}
  n30 = WorldPosition()
  n31 = ScalarParameter() {defaultValue=10000, parameterName=Radius}
  n32 = ScalarParameter() {defaultValue=0.8, parameterName=HArdness}
  NamedRerouteDeclaration(None=StaticSwitchParameter(True=SphereMask(A=@n30.XYZ, B=@n11, Radius=@n31, Hardness=@n32), False=Constant() {r=1}) {defaultValue=true, parameterName=UseWorldLocationBlend}) {name=PP_BlendMask, nodeColor=<rgba 0.6172 0 1 1>}
  n33 = NamedRerouteUsage()
  n34 = Constant2Vector()
  n35 = ScalarParameter() {defaultValue=0.2, parameterName=PrismMaskRadius}
  n36 = ScalarParameter() {defaultValue=0.25, parameterName=PrismMaskHardness}
  NamedRerouteDeclaration(None=OneMinus(None=SphereMask(A=@n33, B=@n34, Radius=@n35, Hardness=@n36))) {name=PrismCenterMask, nodeColor=<rgba 0.8047 0 1 1>}
  NamedRerouteUsage()
  NamedRerouteUsage()
  OneMinus(None=SphereMask(A=@n2, B=Constant2Vector(), Radius=ScalarParameter() {defaultValue=10000, parameterName=Radius}, Hardness=Constant()))
  Power(Base=@n26)
  Subtract(
   A=@n2
   B=MaterialFunctionCall(World Position=@n27) {materialFunction=&TransformToClipSpace}.Buffer UVs for Scene Textures
  )
  TextureCoordinate()
 ]
}
